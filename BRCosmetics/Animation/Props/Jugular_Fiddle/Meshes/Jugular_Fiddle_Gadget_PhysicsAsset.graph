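# Physics asset for the Jugular Fiddle gadget, an animated cosmetic prop made up of a fiddle and a bow.
# Capsule bodies cover the fiddle and bow bones, with a tiny placeholder-sized capsule on the main bone.
# Two constraints link fiddle and bow separately to that main bone, with matching swing and twist limit angles, so neither part is constrained directly to the other.
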PhysicsAsset /BRCosmetics/Animation/Props/Jugular_Fiddle/Meshes/Jugular_Fiddle_Gadget_PhysicsAsset
bodies {
 JNT_Bow [{shapeType=Capsule, length=20.4455, radius=1.3959, center=<-0.0164 16.1882 0>, rotation=<rot 0 -179.9536 -89.985>}]
 JNT_Fiddle [{shapeType=Capsule, length=33.7932, radius=12.4861, center=<0.0141 0.01 0.1674>, rotation=<rot 0 0.036 -87.8254>}]
 JNT_Jugular_Fiddle_Main [{shapeType=Capsule, length=0.505, radius=0.505, rotation=<rot 0 -90 0>}]
}
constraints [
 {bone1Name=JNT_Fiddle, bone2Name=JNT_Jugular_Fiddle_Main, swing1LimitDegrees=45, swing2LimitDegrees=45, twistLimitDegrees=45}
 {bone1Name=JNT_Bow, bone2Name=JNT_Jugular_Fiddle_Main, swing1LimitDegrees=45, swing2LimitDegrees=45, twistLimitDegrees=45}
]
